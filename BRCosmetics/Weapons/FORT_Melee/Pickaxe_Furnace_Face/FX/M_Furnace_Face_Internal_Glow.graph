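# Unlit, two-sided modulate-blend material for the internal glow of the Furnace Face pickaxe, with a tintable color.
# Two smoke texture samples, on bounding-box UVs at different scales with shared rotation and panning, multiply into animated noise.
# Squared vertex-color red, that noise and an inverted Fresnel mask the blend from neutral white to color plus noise.
# The result passes through MF_FX_EmissiveAdjustments.
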Material /BRCosmetics/Weapons/FORT_Melee/Pickaxe_Furnace_Face/FX/M_Furnace_Face_Internal_Glow
graph {
 properties {
  blendMode=BLEND_Modulate
  pixelDepthOffsetMode=PDOM_Legacy
  shadingModel=MSM_Unlit
  translucencyPass=MTP_BeforeDOF
  twoSided=true
  usedShadingModels=Unlit
 }
 nodes [
  n1 = VectorParameter() {parameterName=Color, defaultValue=<rgba 20 3.1012 0 1>}
  n2 = StaticComponentMaskParameter(None=MaterialFunctionCall() {materialFunction=&BoundingBoxBased_0-1_UVW}.RGB) {defaultG=true, defaultR=true, parameterName=Box Mapping Direction}
  n3 = ScalarParameter() {parameterName=Direction}
  n4 = Multiply(A=Time(), B=ScalarParameter() {defaultValue=1, parameterName=Speed Multiplier})
  n5 = Panner(Coordinate=Rotator(Coordinate=Multiply(A=@n2) {constB=0.65}, Time=@n3), Time=@n4) {bFractionalPart=true, speedX=0.2875, speedY=-0.35}
  n6 = Panner(Coordinate=Rotator(Coordinate=Multiply(A=@n2) {constB=0.8}, Time=@n3), Time=@n4) {bFractionalPart=true, speedX=0.2875, speedY=-0.35}
  n7 = Multiply(A=Multiply(
    A=TextureSample(UVs=@n5) {samplerSource=SSM_Wrap_WorldGroupSettings, texture=&T_Inky_Smoke_Tile}.G
    B=TextureSample(UVs=@n6) {samplerSource=SSM_Wrap_WorldGroupSettings, texture=&T_Inky_Smoke_Tile}.G
   )) {constB=2.5}
  n8 = VertexColor()
  n9 = Multiply(A=@n8.R, B=@n8.R)
  n10 = Fresnel(ExponentIn=ScalarParameter() {defaultValue=0.8, parameterName=Fresnel Exponent})
  n11 = Multiply(A=@n7, B=Power(Base=OneMinus(None=@n10), Exp=ScalarParameter() {defaultValue=0.8, parameterName=Fresnel Power}))
  MaterialFunctionCall(Emissive=LinearInterpolate(B=Add(A=@n1.RGB, B=@n7), Alpha=Multiply(A=@n9, B=@n11)) {constA=1}) {materialFunction=&MF_FX_EmissiveAdjustments}
 ]
}
