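# Physics asset for the Furnace Face two-handed pickaxe skeletal mesh, a cosmetic melee weapon.
# A single capsule body on the Melee_TwoHanded_Handle bone approximates the weapon's shape, offset along and rotated to fit the mesh.
# It has no constraints, so the whole pickaxe is represented as one collision body.
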PhysicsAsset /BRCosmetics/Weapons/FORT_Melee/Pickaxe_Furnace_Face/Meshes/SK_Pickaxe_FurnaceFace_PhysicsAsset
bodies {Melee_TwoHanded_Handle [{shapeType=Capsule, length=63.2041, radius=33.7718, center=<0.6816 -1.5359 27.2594>, rotation=<rot 1.1746 -90 -0.0429>}]}
constraints []
